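# Niagara system for the Black Knight Super Leveling cosmetic, revealing variant-tinted sparks and smoke over the character's skeletal mesh.
# A spark burst fires at one bone, while partitioned BoneFlow sprites and per-region swirl-smoke sprites follow the body.
# A sphere mask, with max distance mapped from current level between start and stop levels, gates flow and smoke.
# Integrated mask time replaces particle age, driving fades, material parameters and kills once reached.
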NiagaraSystem /BRCosmetics/Effects/Fort_Effects/Effects/Characters/Athena_Parts/BlackKnight/NS_BlackKnight_SuperLeveling
dependencies {
 usedDataInterfaces [&NiagaraDataInterfaceCurve, &NiagaraDataInterfaceColorCurve]
 usedDynamicInputs [&UniformRangedInt]
 usedModules [&SystemState]
}
emitters {
 Burst {
  emitterSpawnScript [&FN_PreserveSkydiveUserVariables() {module=PreserveUserVariables}]
  emitterUpdateScript [
   &EmitterState()
   &SpawnBurst_Instantaneous(Spawn Count=30, Use Spawn Probability=false)
  ]
  particleSpawnScript [
   /Niagara/Modules/Spawn/Initialization/InitializeParticle(
    Lifetime=&UniformRangedFloat(Minimum=0.7, Maximum=1)
    Module.Write Color=true
    Module.Write Lifetime=true
    Module.Write Mass=false
    Module.Write Material Random=true
    Module.Write Position=false
    Module.Write Scale=false
    Module.Write SpriteRotation=true
    Module.Write SpriteSize=true
    Sprite Rotation=&UniformRangedFloat(Minimum=-360, Maximum=360)
    Sprite Size=&Vector2DFromFloat(Value=&UniformRangedFloat(Minimum=4, Maximum=6))
   )
   /Niagara/Modules/Spawn/Location/SphereLocation(MaskSpawnGroup=false, Module.Override Randomness=false, Surface Only=false)
   &SkeletalMeshLocation(
    Bone / Socket Index=2
    Bone Sampling Mode=::NewEnumerator4
    Inherit Velocity=false
    MaskSpawnGroup=false
    Sampled Position Offset=&Add_Vector(A=$Particles.RandomSphereOffset, B=<0 0 20>)
    Skeletal Mesh=$User.SkeletalMesh
   )
   &AddVelocity(Velocity=&Multiply_VectorByFloat(Vector=$Output.SphereLocation.SphereNormal, Float=&UniformRangedFloat(Minimum=20, Maximum=50)))
  ]
  particleUpdateScript [
   &SolveForcesAndVelocity(Clamp Velocity=false, Limit Acceleration=false)
   &ParticleState()
   /SetVariables(Particles.RandomSphereOffset:Vector3f=&Multiply_VectorByFloat(Vector=$Output.SphereLocation.SphereNormal, Float=&UniformRangedFloat(Minimum=10, Maximum=16))) {isSetParameters=true, module=SetVariables_#1}
   &ScaleSpriteSize(Scale Factor=&Vector2DFromFloat(Value=&FloatFromCurve(FloatCurve=DataInterface() {propertyValues {Curve {keys [<at 0 1>,<at 1 0>]}}})))
   &ScaleColor(
    Color Value To Scale=$Particles.Color
    Scale Alpha=&FloatFromCurve(FloatCurve=DataInterface() {propertyValues {Curve {keys [
         <at 0 0 RCIM_Cubic>
         <at 0.25 1 RCIM_Cubic>
         {arriveTangent=-5.4092, interpMode=RCIM_Cubic, leaveTangent=-5.4092, tangentMode=RCTM_User, time=0.9, value=0}
        ]}}})
    Scale RGB=&MakeVectorFromLinearColorRGB(Color=&ColorFromCurve(ColorCurve=DataInterface() {propertyValues {
        AlphaCurve {keys [<at 0 1 RCIM_Cubic>,<at 1 0 RCIM_Cubic>]}
        BlueCurve {keys [<at 0 70 RCIM_Cubic>,<at 1 1 RCIM_Cubic>]}
        GreenCurve {keys [<at 0 70 RCIM_Cubic>,<at 1 1 RCIM_Cubic>]}
        RedCurve {keys [<at 0 70 RCIM_Cubic>,<at 1 1 RCIM_Cubic>]}
       }}))
    ScaleA=true
    ScaleRGB=true
   )
   &ScaleColor(Scale RGB=&MakeVectorFromLinearColorRGB(Color=$User.BKSL_VariantColor), ScaleA=true, ScaleRGB=true) {module=ScaleColor001}
  ]
  rendererClasses [&NiagaraSpriteRendererProperties]
  renderers [{rendererClass=&NiagaraSpriteRendererProperties}]
 }
 FN_BoneFlow {
  emitterUpdateScript [
   &EmitterState()
   /Game/Effects/Niagara/Modules/Partition/FN_PartitionAttributes/FN_PartitionAttributes_v2(Particles Per Partition=35, Partitions=5)
   &SpawnBurst_Instantaneous(Spawn Count=$Emitter.Spawn Total, Spawn Probability=HlslExpression() {hlslExpression="Emitter.LoopCount == 0"}, Use Spawn Probability=true)
  ]
  particleSpawnScript [
   /Game/Effects/Niagara/Modules/Partition/FN_ParticlePartition_v2/FN_ParticlePartition_v2()
   &InitializeParticle(
    Color=<rgba 2 2 2 1>
    Lifetime=&UniformRangedFloat(Minimum=1.2, Maximum=2)
    Module.Write Color=true
    Module.Write Lifetime=true
    Module.Write Mass=true
    Module.Write Material Random=true
    Module.Write Position=false
    Module.Write Scale=false
    Module.Write SpriteRotation=true
    Module.Write SpriteSize=true
    Sprite Rotation=&UniformRangedFloat(Minimum=-360, Maximum=360)
    Sprite Size=&Vector2DFromFloat(Value=&UniformRangedFloat(Minimum=4, Maximum=6))
   )
   /SetVariables(Particles.BoneFlowArmsandLegs:NiagaraFloat=&Lerp_Float(A=0.85, B=0.45, Alpha=$Particles.T)) {isSetParameters=true}
   /SetVariables(Particles.BoneFlowChest:NiagaraFloat=&Lerp_Float(A=1.5, B=0.85, Alpha=$Particles.T)) {isSetParameters=true}
   /Game/Effects/Niagara/Modules/BoneFlow/FN_Initialize_BoneFlow_Attributes_v2(
    Angle:NiagaraFloat=&UniformRangedFloat(Minimum=0, Maximum=6)
    Radius:NiagaraFloat=&Multiply_Float(
     A=HlslExpression() {hlslExpression="Particles.Partition.ID > 5 ? 50. : 10."}
     B=&Lerp_Float(
      A=&Lerp_Float(A=1.95, B=0.95, Alpha=$Particles.T)
      B=&Lerp_Float(A=0.85, B=0.45, Alpha=$Particles.T)
      Alpha=HlslExpression() {hlslExpression="Particles.Partition.ID < 2 ? 3 : 1"}
     )
    )
   )
   /SetVariables(Particles.EmbersSubImageIndex:NiagaraFloat=&UniformRangedFloat(Minimum=0, Maximum=35)) {isSetParameters=true}
  ]
  particleUpdateScript [
   &ParticleState(Kill Particles When Lifetime Has Elapsed=false, Let Infinitely Lived Particles Die When Emitter Deactivates=true)
   /Game/Effects/Niagara/Modules/BoneFlow/FN_BoneFlow_v2/FN_BoneFlow_v2(
    Angle:NiagaraFloat=$Particles.FN_Initialize_BoneFlow_Attributes_v2.Angle
    Radius:NiagaraFloat=$Particles.FN_Initialize_BoneFlow_Attributes_v2.Radius
    SkeletalMesh:NiagaraDataInterfaceSkeletalMesh=$User.SkeletalMesh
    T:NiagaraFloat=$Particles.FN_Initialize_BoneFlow_Attributes_v2.T
   )
   /SetVariables(Particles.SphereMaskOrigin:Vector3f=&Add_Vector(A=<0 0 120>, B=Linked() {linkedVariable {name=Engine.Owner.Position, type=NiagaraPosition}})) {isSetParameters=true}
   /Game/Effects/Niagara/Modules/Color/SphereMask(
    Max Distance:NiagaraFloat=/Game/Effects/Niagara/DynamicInputs/Math/FN_Fit(
     Clamp:NiagaraBool=false
     Value:NiagaraFloat=$User.BKSL_CurrentLevel
     SourceMin:NiagaraFloat=$User.BKSL_StartAtLevel
     SourceMax:NiagaraFloat=$User.BKSL_StopAtLevel
     DestMin:NiagaraFloat=8
     DestMax:NiagaraFloat=125
    )
    Particle Position:NiagaraPosition=$Particles.Position
    Speed=35
    Sphere Center:Vector3f=$Particles.SphereMaskOrigin
   )
   /Game/Effects/Niagara/Modules/Math/FN_Integrator/FN_Integrator(Input=HlslExpression() {hlslExpression=Particles.SphereMask * Engine.DeltaTime})
   /SetVariables(Particles.NormalizedLife=HlslExpression() {hlslExpression="max(min(Particles.FN_Integrator.Attr / Particles.Lifetime,1.0),0)"}) {isSetParameters=true}
   &Color(Color=&LerpLinearColors(
     StartColor=&MultiplyLinearColor(
      Linear Color=$Particles.Initial.Color
      Scale Factor=&MakeVector4_FromLinearColor(LinearColor=$User.BKSL_VariantColor)
     )
     EndColor=&MultiplyLinearColor(
      Linear Color=$Particles.Initial.Color
      Scale Factor=&MakeVector4_FromLinearColor(LinearColor=$User.BKSL_VariantColor)
     )
     LerpFactor=&MakeFloatFromInt(Int=$Particles.SphereMask)
    ))
   &ScaleSpriteSize(Scale Factor=&Vector2DFromFloat(Value=&FloatFromCurve(FloatCurve=DataInterface() {propertyValues {Curve {keys [<at 0 1>,<at 1 0>]}}}, CurveIndex=$Particles.NormalizedLife)))
   &ScaleColor(
    Color Value To Scale=$Particles.Color
    Scale Alpha=&FloatFromCurve(
     FloatCurve=DataInterface() {propertyValues {Curve {keys [
         <at 0 0 RCIM_Cubic>
         <at 0.1 1 RCIM_Cubic>
         {arriveTangent=-2.6078, interpMode=RCIM_Cubic, leaveTangent=-2.6078, tangentMode=RCTM_User, time=0.9, value=0}
        ]}}}
     CurveIndex=$Particles.NormalizedLife
    )
    Scale RGB=&MakeVectorFromLinearColorRGB(Color=&ColorFromCurve(
      ColorCurve=DataInterface() {propertyValues {
        AlphaCurve {keys [<at 0 1 RCIM_Cubic>,<at 1 0 RCIM_Cubic>]}
        BlueCurve {keys [<at 0 60 RCIM_Cubic>,<at 1 1 RCIM_Cubic>]}
        GreenCurve {keys [<at 0 60 RCIM_Cubic>,<at 1 1 RCIM_Cubic>]}
        RedCurve {keys [<at 0 60 RCIM_Cubic>,<at 1 1 RCIM_Cubic>]}
       }}
      CurveIndex=$Particles.NormalizedLife
     ))
    ScaleA=true
    ScaleRGB=true
   )
   &KillParticles(Kill Particles=HlslExpression() {hlslExpression="Particles.NormalizedLife == 1.0 ? 1 : 0"})
  ]
  rendererClasses [&NiagaraSpriteRendererProperties]
  renderers [{rendererClass=&NiagaraSpriteRendererProperties}]
 }
 "SwirlSmokeArms&&Legs" {
  emitterSpawnScript [&FN_PreserveSkydiveUserVariables() {module=PreserveUserVariables}]
  emitterUpdateScript [
   &EmitterState()
   &SpawnBurst_Instantaneous(Spawn Count=12, Spawn Probability=HlslExpression() {hlslExpression="Emitter.LoopCount == 0"}, Use Spawn Probability=true)
  ]
  particleSpawnScript [
   /Game/Effects/Niagara/Modules/Groups/FN_SetSpawnGroup_v1/FN_SetSpawnGroup(Override SpawnGroup=true, SpawnGroup=HlslExpression() {hlslExpression="Modulo(Particles.UniqueID, 12) + 3"})
   &InitializeParticle(
    Color=<rgba 1 1 1 0>
    Lifetime=&UniformRangedFloat(Minimum=1, Maximum=2)
    Module.Write Color=true
    Module.Write Lifetime=true
    Module.Write Mass=false
    Module.Write Material Random=true
    Module.Write Position=false
    Module.Write Scale=false
    Module.Write SpriteRotation=true
    Module.Write SpriteSize=true
    Sprite Rotation=&UniformRangedFloat(Minimum=-360, Maximum=360)
    Sprite Size=&Vector2DFromFloat(Value=&UniformRangedFloat(Minimum=36, Maximum=46))
   )
   /SetVariables(Particles.RandomStartingFrame:NiagaraFloat=&UniformRangedFloat(Minimum=0, Maximum=64)) {isSetParameters=true, module=SetVariables_#2}
   /SetVariables(Particles.RandomSphericalOffset:NiagaraPosition=$Particles.Position) {isSetParameters=true, module=SetVariables_#3}
   /SetVariables(Particles.SphereMaskOrigin:Vector3f=<0 0 120>) {isSetParameters=true, module=SetVariables_#4}
   &SphereLocation(MaskSpawnGroup=false, Module.Override Randomness=false, Surface Only=false)
   /SetVariables(Particles.RandomSphereOffset:Vector3f=&Multiply_VectorByFloat(Vector=$Output.SphereLocation.SphereNormal, Float=3)) {isSetParameters=true, module=SetVariables_#1}
   /SetVariables(Particles.SmokeColorRange:LinearColor=&UniformRangedLinearColor(Minimum=<rgba 0.5 0.5 0.5 0.2>, Maximum=<rgba 0.8 0.8 0.8 0.3>)) {isSetParameters=true, module=SetVariables_#5}
   &CameraOffset(Camera Offset Amount=&UniformRangedFloat(Minimum=-10, Maximum=30))
  ]
  particleUpdateScript [
   &ParticleState(Kill Particles When Lifetime Has Elapsed=false, Let Infinitely Lived Particles Die When Emitter Deactivates=true)
   &SkeletalMeshLocation(
    Bone / Socket Index=$Particles.SpawnGroup
    Bone Sampling Mode=::NewEnumerator4
    Inherit Velocity=false
    MaskSpawnGroup=false
    Sampled Position Offset=$Particles.RandomSphereOffset
    Skeletal Mesh=$User.SkeletalMesh
   )
   &SphereMask(
    Max Distance:NiagaraFloat=&FN_Fit(
     Clamp:NiagaraBool=false
     Value:NiagaraFloat=$User.BKSL_CurrentLevel
     SourceMin:NiagaraFloat=$User.BKSL_StartAtLevel
     SourceMax:NiagaraFloat=$User.BKSL_StopAtLevel
     DestMin:NiagaraFloat=0
     DestMax:NiagaraFloat=125
    )
    Particle Position:Vector3f=$Output.SkeletalMeshLocation.SampledPosition
    Speed=50
    Sphere Center:Vector3f=$Particles.Initial.SphereMaskOrigin
   )
   &FN_Integrator(Input=HlslExpression() {hlslExpression=Particles.SphereMask * Engine.DeltaTime})
   /SetVariables(Particles.NormalizedLife=HlslExpression() {hlslExpression="max(min(Particles.FN_Integrator.Attr / Particles.Lifetime,1.0),0)"}) {isSetParameters=true, module=SetVariables_#6}
   &KillParticles(Kill Particles=HlslExpression() {hlslExpression="Particles.NormalizedLife == 1.0 ? 1 : 0"})
   &Color(Color=&LerpLinearColors(
     StartColor=$Particles.Initial.Color
     EndColor=&MultiplyLinearColor(
      Linear Color=$Particles.Initial.SmokeColorRange
      Scale Factor=&MakeVector4_FromLinearColor(LinearColor=$User.BKSL_VariantColor)
     )
     LerpFactor=&MakeFloatFromInt(Int=$Particles.SphereMask)
    ))
   &ScaleColor(
    Color Value To Scale=$Particles.Color
    Scale Alpha=&FloatFromCurve(
     FloatCurve=DataInterface() {propertyValues {Curve {keys [
         <at 0 0 RCIM_Cubic>
         <at 0.1 1 RCIM_Cubic>
         {arriveTangent=-5.4092, interpMode=RCIM_Cubic, leaveTangent=-5.4092, tangentMode=RCTM_User, time=1, value=0}
        ]}}}
     CurveIndex=$Particles.NormalizedLife
    )
    ScaleA=true
    ScaleRGB=false
   )
   /Niagara/Modules/Update/SubUV/SubUVAnimation(
    Number Of Frames=64
    SubUV Lookup Index=&Multiply_Float(A=&Add_Float(A=$Particles.NormalizedLife, B=$Particles.RandomStartingFrame), B=0.75)
   )
   &DynamicMaterialParameters(
    Index 0 Param 1=&FloatFromCurve(FloatCurve=DataInterface() {propertyValues {Curve {keys [<at 0.3 1 RCIM_Cubic>,<at 1 -0.1>]}}}, CurveIndex=$Particles.NormalizedLife)
    Index 0 Param 2=&FloatFromCurve(
     FloatCurve=DataInterface() {propertyValues {Curve {keys [<at 0 1 RCIM_Cubic>,<at 0.8 8.8 RCIM_Cubic>]}}}
     CurveIndex=$Particles.NormalizedLife
    )
    Index 0 Param 4=&FloatFromCurve(FloatCurve=DataInterface() {propertyValues {Curve {keys [<at 0.3 0>,<at 0.5 0.6>]}}}, CurveIndex=$Particles.NormalizedLife)
    Module.Param0WriteEnabled=true
    Module.Param1WriteEnabled=true
    Module.Param2WriteEnabled=true
    Module.Param3WriteEnabled=true
   )
  ]
  rendererClasses [&NiagaraSpriteRendererProperties]
  renderers [
   {rendererClass=&NiagaraSpriteRendererProperties}
   {rendererClass=&NiagaraSpriteRendererProperties, rendererIndex=1}
  ]
 }
 SwirlSmokeChest {
  emitterSpawnScript [&FN_PreserveSkydiveUserVariables() {module=PreserveUserVariables}]
  emitterUpdateScript [
   &EmitterState()
   &SpawnBurst_Instantaneous(Spawn Count=6, Spawn Probability=HlslExpression() {hlslExpression="Emitter.LoopCount == 0"}, Use Spawn Probability=true)
  ]
  particleSpawnScript [
   &FN_SetSpawnGroup(Override SpawnGroup=true, SpawnGroup=HlslExpression() {hlslExpression="Modulo(Particles.UniqueID, 3) "})
   &InitializeParticle(
    Color=<rgba 1 1 1 0>
    Lifetime=&UniformRangedFloat(Minimum=2, Maximum=2.5)
    Module.Write Color=true
    Module.Write Lifetime=true
    Module.Write Mass=false
    Module.Write Material Random=true
    Module.Write Position=false
    Module.Write Scale=false
    Module.Write SpriteRotation=true
    Module.Write SpriteSize=true
    Sprite Rotation=&UniformRangedFloat(Minimum=-360, Maximum=360)
    Sprite Size=&Vector2DFromFloat(Value=&UniformRangedFloat(Minimum=40, Maximum=45))
   )
   /SetVariables(Particles.RandomStartingFrame:NiagaraFloat=&UniformRangedFloat(Minimum=0, Maximum=64)) {isSetParameters=true, module=SetVariables_#2}
   /SetVariables(Particles.RandomSphericalOffset:NiagaraPosition=$Particles.Position) {isSetParameters=true, module=SetVariables_#3}
   /SetVariables(Particles.SphereMaskOrigin:Vector3f=<0 0 120>) {isSetParameters=true, module=SetVariables_#4}
   &SphereLocation(MaskSpawnGroup=false, Module.Override Randomness=false, Surface Only=false)
   /SetVariables(Particles.RandomSphereOffset:Vector3f=&Multiply_VectorByFloat(Vector=$Output.SphereLocation.SphereNormal, Float=5)) {isSetParameters=true, module=SetVariables_#1}
   /SetVariables(Particles.SmokeColorRange:LinearColor=&UniformRangedLinearColor(Minimum=<rgba 0.5 0.5 0.5 0.2>, Maximum=<rgba 0.8 0.8 0.8 0.3>)) {isSetParameters=true, module=SetVariables_#5}
   &CameraOffset(Camera Offset Amount=&UniformRangedFloat(Minimum=-10, Maximum=30))
  ]
  particleUpdateScript [
   &ParticleState(Kill Particles When Lifetime Has Elapsed=false, Let Infinitely Lived Particles Die When Emitter Deactivates=true)
   &SkeletalMeshLocation(
    Bone / Socket Index=$Particles.SpawnGroup
    Bone Sampling Mode=::NewEnumerator4
    Inherit Velocity=false
    MaskSpawnGroup=false
    Sampled Position Offset=&Add_Vector(A=$Particles.RandomSphereOffset, B=<0 0 -5>)
    Skeletal Mesh=$User.SkeletalMesh
   )
   &SphereMask(
    Max Distance:NiagaraFloat=&FN_Fit(
     Clamp:NiagaraBool=false
     Value:NiagaraFloat=$User.BKSL_CurrentLevel
     SourceMin:NiagaraFloat=$User.BKSL_StartAtLevel
     SourceMax:NiagaraFloat=$User.BKSL_StopAtLevel
     DestMin:NiagaraFloat=10
     DestMax:NiagaraFloat=125
    )
    Particle Position:Vector3f=$Output.SkeletalMeshLocation.SampledPosition
    Speed=50
    Sphere Center:Vector3f=$Particles.Initial.SphereMaskOrigin
   )
   &FN_Integrator(Input=HlslExpression() {hlslExpression=Particles.SphereMask * Engine.DeltaTime})
   /SetVariables(Particles.NormalizedLife=HlslExpression() {hlslExpression="max(min(Particles.FN_Integrator.Attr / Particles.Lifetime,1.0),0)"}) {isSetParameters=true, module=SetVariables_#6}
   &KillParticles(Kill Particles=HlslExpression() {hlslExpression="Particles.NormalizedLife == 1.0 ? 1 : 0"})
   &Color(Color=&LerpLinearColors(
     StartColor=$Particles.Initial.Color
     EndColor=&MultiplyLinearColor(
      Linear Color=$Particles.Initial.SmokeColorRange
      Scale Factor=&MakeVector4_FromLinearColor(LinearColor=$User.BKSL_VariantColor)
     )
     LerpFactor=&MakeFloatFromInt(Int=$Particles.SphereMask)
    ))
   &ScaleColor(
    Color Value To Scale=$Particles.Color
    Scale Alpha=&FloatFromCurve(
     FloatCurve=DataInterface() {propertyValues {Curve {keys [
         <at 0 0 RCIM_Cubic>
         <at 0.1 1.25 RCIM_Cubic>
         {arriveTangent=-4.6509, interpMode=RCIM_Cubic, leaveTangent=-4.651, tangentMode=RCTM_User, time=1, value=0}
        ]}}}
     CurveIndex=$Particles.NormalizedLife
    )
    ScaleA=true
    ScaleRGB=false
   )
   &SubUVAnimation(
    Number Of Frames=64
    SubUV Lookup Index=&Multiply_Float(A=&Add_Float(A=$Particles.NormalizedLife, B=$Particles.RandomStartingFrame), B=0.75)
   )
   &DynamicMaterialParameters(
    Index 0 Param 1=&FloatFromCurve(FloatCurve=DataInterface() {propertyValues {Curve {keys [<at 0.3 1 RCIM_Cubic>,<at 1 -0.1>]}}}, CurveIndex=$Particles.NormalizedLife)
    Index 0 Param 2=&FloatFromCurve(
     FloatCurve=DataInterface() {propertyValues {Curve {keys [<at 0 1 RCIM_Cubic>,<at 0.8 8.8 RCIM_Cubic>]}}}
     CurveIndex=$Particles.NormalizedLife
    )
    Index 0 Param 4=&FloatFromCurve(FloatCurve=DataInterface() {propertyValues {Curve {keys [<at 0.3 0>,<at 0.5 0.6>]}}}, CurveIndex=$Particles.NormalizedLife)
    Module.Param0WriteEnabled=true
    Module.Param1WriteEnabled=true
    Module.Param2WriteEnabled=true
    Module.Param3WriteEnabled=true
   )
  ]
  rendererClasses [&NiagaraSpriteRendererProperties]
  renderers [
   {rendererClass=&NiagaraSpriteRendererProperties}
   {rendererClass=&NiagaraSpriteRendererProperties, rendererIndex=1}
  ]
 }
 SwirlSmokeHead {
  emitterSpawnScript [&FN_PreserveSkydiveUserVariables() {module=PreserveUserVariables}]
  emitterUpdateScript [
   &EmitterState()
   &SpawnBurst_Instantaneous(Spawn Count=2, Spawn Probability=HlslExpression() {hlslExpression="Emitter.LoopCount == 0"}, Use Spawn Probability=true)
  ]
  particleSpawnScript [
   &FN_SetSpawnGroup(Override SpawnGroup=true, SpawnGroup=HlslExpression() {hlslExpression="Modulo(Particles.UniqueID, 3) +18"})
   &InitializeParticle(
    Color=<rgba 1 1 1 0>
    Lifetime=&UniformRangedFloat(Minimum=1.5, Maximum=2)
    Module.Write Color=true
    Module.Write Lifetime=true
    Module.Write Mass=false
    Module.Write Material Random=true
    Module.Write Position=false
    Module.Write Scale=false
    Module.Write SpriteRotation=true
    Module.Write SpriteSize=true
    Sprite Rotation=&UniformRangedFloat(Minimum=-360, Maximum=360)
    Sprite Size=&Vector2DFromFloat(Value=&UniformRangedFloat(Minimum=32, Maximum=45))
   )
   /SetVariables(Particles.RandomStartingFrame:NiagaraFloat=&UniformRangedFloat(Minimum=0, Maximum=64)) {isSetParameters=true, module=SetVariables_#2}
   /SetVariables(Particles.RandomSphericalOffset:NiagaraPosition=$Particles.Position) {isSetParameters=true, module=SetVariables_#3}
   /SetVariables(Particles.SphereMaskOrigin:Vector3f=<0 0 120>) {isSetParameters=true, module=SetVariables_#4}
   &SphereLocation(MaskSpawnGroup=false, Module.Override Randomness=false, Surface Only=false)
   /SetVariables(Particles.RandomSphereOffset:Vector3f=&Multiply_VectorByFloat(Vector=$Output.SphereLocation.SphereNormal, Float=5)) {isSetParameters=true, module=SetVariables_#1}
   /SetVariables(Particles.SmokeColorRange:LinearColor=&UniformRangedLinearColor(Minimum=<rgba 0.5 0.5 0.5 0.2>, Maximum=<rgba 0.8 0.8 0.8 0.3>)) {isSetParameters=true, module=SetVariables_#5}
   &CameraOffset(Camera Offset Amount=&UniformRangedFloat(Minimum=-10, Maximum=30))
  ]
  particleUpdateScript [
   &ParticleState(Kill Particles When Lifetime Has Elapsed=false, Let Infinitely Lived Particles Die When Emitter Deactivates=true)
   &SkeletalMeshLocation(
    Bone / Socket Index=$Particles.SpawnGroup
    Bone Sampling Mode=::NewEnumerator4
    Inherit Velocity=false
    MaskSpawnGroup=false
    Sampled Position Offset=&Add_Vector(A=$Particles.RandomSphereOffset, B=<0 0 15>)
    Skeletal Mesh=$User.SkeletalMesh
   )
   &SphereMask(
    Max Distance:NiagaraFloat=&FN_Fit(
     Clamp:NiagaraBool=false
     Value:NiagaraFloat=$User.BKSL_CurrentLevel
     SourceMin:NiagaraFloat=$User.BKSL_StartAtLevel
     SourceMax:NiagaraFloat=$User.BKSL_StopAtLevel
     DestMin:NiagaraFloat=0
     DestMax:NiagaraFloat=125
    )
    Particle Position:Vector3f=$Output.SkeletalMeshLocation.SampledPosition
    Speed=50
    Sphere Center:Vector3f=$Particles.Initial.SphereMaskOrigin
   )
   &FN_Integrator(Input=HlslExpression() {hlslExpression=Particles.SphereMask * Engine.DeltaTime})
   /SetVariables(Particles.NormalizedLife=HlslExpression() {hlslExpression="max(min(Particles.FN_Integrator.Attr / Particles.Lifetime,1.0),0)"}) {isSetParameters=true, module=SetVariables_#6}
   &KillParticles(Kill Particles=HlslExpression() {hlslExpression="Particles.NormalizedLife == 1.0 ? 1 : 0"})
   &Color(Color=&LerpLinearColors(
     StartColor=$Particles.Initial.Color
     EndColor=&MultiplyLinearColor(
      Linear Color=$Particles.Initial.SmokeColorRange
      Scale Factor=&MakeVector4_FromLinearColor(LinearColor=$User.BKSL_VariantColor)
     )
     LerpFactor=&MakeFloatFromInt(Int=$Particles.SphereMask)
    ))
   &ScaleColor(
    Color Value To Scale=$Particles.Color
    Scale Alpha=&FloatFromCurve(
     FloatCurve=DataInterface() {propertyValues {Curve {keys [
         <at 0 0 RCIM_Cubic>
         <at 0.1 1 RCIM_Cubic>
         {arriveTangent=-5.4092, interpMode=RCIM_Cubic, leaveTangent=-5.4092, tangentMode=RCTM_User, time=1, value=0}
        ]}}}
     CurveIndex=$Particles.NormalizedLife
    )
    ScaleA=true
    ScaleRGB=false
   )
   &SubUVAnimation(
    Number Of Frames=64
    SubUV Lookup Index=&Multiply_Float(A=&Add_Float(A=$Particles.NormalizedLife, B=$Particles.RandomStartingFrame), B=0.75)
   )
   &DynamicMaterialParameters(
    Index 0 Param 1=&FloatFromCurve(FloatCurve=DataInterface() {propertyValues {Curve {keys [<at 0.3 1 RCIM_Cubic>,<at 1 -0.1>]}}}, CurveIndex=$Particles.NormalizedLife)
    Index 0 Param 2=&FloatFromCurve(
     FloatCurve=DataInterface() {propertyValues {Curve {keys [<at 0 1 RCIM_Cubic>,<at 0.8 8.8 RCIM_Cubic>]}}}
     CurveIndex=$Particles.NormalizedLife
    )
    Index 0 Param 4=&FloatFromCurve(FloatCurve=DataInterface() {propertyValues {Curve {keys [<at 0.3 0>,<at 0.5 0.6>]}}}, CurveIndex=$Particles.NormalizedLife)
    Module.Param0WriteEnabled=true
    Module.Param1WriteEnabled=true
    Module.Param2WriteEnabled=true
    Module.Param3WriteEnabled=true
   )
  ]
  rendererClasses [&NiagaraSpriteRendererProperties]
  renderers [
   {rendererClass=&NiagaraSpriteRendererProperties}
   {rendererClass=&NiagaraSpriteRendererProperties, rendererIndex=1}
  ]
 }
 SwirlSmokeThighs {
  emitterSpawnScript [&FN_PreserveSkydiveUserVariables() {module=PreserveUserVariables}]
  emitterUpdateScript [
   &EmitterState()
   &SpawnBurst_Instantaneous(Spawn Count=6, Spawn Probability=HlslExpression() {hlslExpression="Emitter.LoopCount == 0"}, Use Spawn Probability=true)
  ]
  particleSpawnScript [
   &FN_SetSpawnGroup(Override SpawnGroup=true, SpawnGroup=HlslExpression() {hlslExpression="Modulo(Particles.UniqueID, 6) + 21"})
   &InitializeParticle(
    Color=<rgba 1 1 1 0>
    Lifetime=&UniformRangedFloat(Minimum=1.5, Maximum=2)
    Module.Write Color=true
    Module.Write Lifetime=true
    Module.Write Mass=false
    Module.Write Material Random=true
    Module.Write Position=false
    Module.Write Scale=false
    Module.Write SpriteRotation=true
    Module.Write SpriteSize=true
    Sprite Rotation=&UniformRangedFloat(Minimum=-360, Maximum=360)
    Sprite Size=&Vector2DFromFloat(Value=&UniformRangedFloat(Minimum=40, Maximum=45))
   )
   /SetVariables(Particles.RandomStartingFrame:NiagaraFloat=&UniformRangedFloat(Minimum=0, Maximum=64)) {isSetParameters=true, module=SetVariables_#2}
   /SetVariables(Particles.RandomSphericalOffset:NiagaraPosition=$Particles.Position) {isSetParameters=true, module=SetVariables_#3}
   /SetVariables(Particles.SphereMaskOrigin:Vector3f=<0 0 120>) {isSetParameters=true, module=SetVariables_#4}
   &SphereLocation(MaskSpawnGroup=false, Module.Override Randomness=false, Surface Only=false)
   /SetVariables(Particles.RandomSphereOffset:Vector3f=&Multiply_VectorByFloat(Vector=$Output.SphereLocation.SphereNormal, Float=5)) {isSetParameters=true, module=SetVariables_#1}
   /SetVariables(Particles.SmokeColorRange:LinearColor=&UniformRangedLinearColor(Minimum=<rgba 0.5 0.5 0.5 0.2>, Maximum=<rgba 0.8 0.8 0.8 0.3>)) {isSetParameters=true, module=SetVariables_#5}
   &CameraOffset(Camera Offset Amount=&UniformRangedFloat(Minimum=-10, Maximum=30))
  ]
  particleUpdateScript [
   &ParticleState(Kill Particles When Lifetime Has Elapsed=false, Let Infinitely Lived Particles Die When Emitter Deactivates=true)
   &SkeletalMeshLocation(
    Bone / Socket Index=$Particles.SpawnGroup
    Bone Sampling Mode=::NewEnumerator4
    Inherit Velocity=false
    MaskSpawnGroup=false
    Sampled Position Offset=&Add_Vector(A=$Particles.RandomSphereOffset, B=<0 0 0>)
    Skeletal Mesh=$User.SkeletalMesh
   )
   &SphereMask(
    Max Distance:NiagaraFloat=&FN_Fit(
     Clamp:NiagaraBool=false
     Value:NiagaraFloat=$User.BKSL_CurrentLevel
     SourceMin:NiagaraFloat=$User.BKSL_StartAtLevel
     SourceMax:NiagaraFloat=$User.BKSL_StopAtLevel
     DestMin:NiagaraFloat=-20
     DestMax:NiagaraFloat=125
    )
    Particle Position:Vector3f=$Output.SkeletalMeshLocation.SampledPosition
    Speed=50
    Sphere Center:Vector3f=$Particles.Initial.SphereMaskOrigin
   )
   &FN_Integrator(Input=HlslExpression() {hlslExpression=Particles.SphereMask * Engine.DeltaTime})
   /SetVariables(Particles.NormalizedLife=HlslExpression() {hlslExpression="max(min(Particles.FN_Integrator.Attr / Particles.Lifetime,1.0),0)"}) {isSetParameters=true, module=SetVariables_#6}
   &KillParticles(Kill Particles=HlslExpression() {hlslExpression="Particles.NormalizedLife == 1.0 ? 1 : 0"})
   &Color(Color=&LerpLinearColors(
     StartColor=$Particles.Initial.Color
     EndColor=&MultiplyLinearColor(
      Linear Color=$Particles.Initial.SmokeColorRange
      Scale Factor=&MakeVector4_FromLinearColor(LinearColor=$User.BKSL_VariantColor)
     )
     LerpFactor=&MakeFloatFromInt(Int=$Particles.SphereMask)
    ))
   &ScaleColor(
    Color Value To Scale=$Particles.Color
    Scale Alpha=&FloatFromCurve(
     FloatCurve=DataInterface() {propertyValues {Curve {keys [
         <at 0 0 RCIM_Cubic>
         <at 0.1 1 RCIM_Cubic>
         {arriveTangent=-5.4092, interpMode=RCIM_Cubic, leaveTangent=-5.4092, tangentMode=RCTM_User, time=1, value=0}
        ]}}}
     CurveIndex=$Particles.NormalizedLife
    )
    ScaleA=true
    ScaleRGB=false
   )
   &SubUVAnimation(
    Number Of Frames=64
    SubUV Lookup Index=&Multiply_Float(A=&Add_Float(A=$Particles.NormalizedLife, B=$Particles.RandomStartingFrame), B=0.75)
   )
   &DynamicMaterialParameters(
    Index 0 Param 1=&FloatFromCurve(FloatCurve=DataInterface() {propertyValues {Curve {keys [<at 0.3 1 RCIM_Cubic>,<at 1 -0.1>]}}}, CurveIndex=$Particles.NormalizedLife)
    Index 0 Param 2=&FloatFromCurve(
     FloatCurve=DataInterface() {propertyValues {Curve {keys [<at 0 1 RCIM_Cubic>,<at 0.8 8.8 RCIM_Cubic>]}}}
     CurveIndex=$Particles.NormalizedLife
    )
    Index 0 Param 4=&FloatFromCurve(FloatCurve=DataInterface() {propertyValues {Curve {keys [<at 0.3 0>,<at 0.5 0.6>]}}}, CurveIndex=$Particles.NormalizedLife)
    Module.Param0WriteEnabled=true
    Module.Param1WriteEnabled=true
    Module.Param2WriteEnabled=true
    Module.Param3WriteEnabled=true
   )
  ]
  rendererClasses [&NiagaraSpriteRendererProperties]
  renderers [
   {rendererClass=&NiagaraSpriteRendererProperties}
   {rendererClass=&NiagaraSpriteRendererProperties, rendererIndex=1}
  ]
 }
}
summary {
 systemName=NS_BlackKnight_SuperLeveling
 emitters [
  {emitterName="SwirlSmokeArms&&Legs", rendererClasses [&NiagaraSpriteRendererProperties]}
  {emitterName=SwirlSmokeChest, rendererClasses [&NiagaraSpriteRendererProperties]}
  {emitterName=SwirlSmokeHead, rendererClasses [&NiagaraSpriteRendererProperties]}
  {emitterName=FN_BoneFlow, rendererClasses [&NiagaraSpriteRendererProperties]}
  {emitterName=SwirlSmokeThighs, rendererClasses [&NiagaraSpriteRendererProperties]}
  {emitterName=Burst, rendererClasses [&NiagaraSpriteRendererProperties]}
 ]
 userVariables [
  User.bIsFrontEndPreview {type=NiagaraBool}
  User.BKSL_CurrentLevel {type=NiagaraFloat, defaultValue=160, description=""}
  User.BKSL_StartAtLevel {type=NiagaraFloat, defaultValue=120, description=""}
  User.BKSL_StopAtLevel {type=NiagaraFloat, defaultValue=160, description=""}
  User.BKSL_VariantColor {type=LinearColor, description="", defaultValue {struct=LinearColor, value=<rgba 0.85 0.85 0.85 1>}}
  User.PlayerSpeed {type=NiagaraFloat, defaultValue=0, description=""}
  User.SkeletalMesh {
   type=NiagaraDataInterfaceSkeletalMesh
   defaultValue {
    struct=NiagaraExt_VariableValue_DataInterface
    value {dataInterface=&self:NiagaraDataInterfaceSkeletalMesh_0}
   }
  }
 ]
}
